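# Editor utility widget for perf cleanup that finds level actors with non-Static, non-editor-only static mesh components and resets their mobility.
# Find lists each actor with its scene component count; right-click focuses, selects and prints component mobilities, while Shift+right-click locates its class asset.
# One fix resets root Mobility on BuildingActors with under four scene components.
# Another resets every scene component on found actors, using ResetEditorProperty to restore defaults.
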EditorUtilityWidgetBlueprint /DelMarTools/EditorUtilities/PerfTools/MobilityResetTool/EUW_MobilityResetTool
graphs [
 EventGraph {nodes [
   "On Clicked (btnFind)"!() -> (OutputDelegate, then, widget:User Widget Object Reference, event:Pointer Event Structure) {is=title K2Node_ComponentBoundEvent}
   > clearItems(self:TA Generic 01 List Object Reference=Get TA_Generic01_List() -> (TA_Generic01_List:TA Generic 01 List Object Reference))
   > n1 = FindActorsWithMoveableComponents!(execute, self) -> (then, actors:Array of Actor Object References)
   > n2 = For Each Loop(Array:Array of Actor Object References=@n1) -> (Array Element:Actor Object Reference)
   n3 = getListItems!(execute=@n2.Completed, self:TA Generic 01 List Object Reference=Get TA_Generic01_List()) -> (then, items:Array of User Widget Object References)
   n4 = "To String (Integer)"(InInt=Length(TargetArray:Array of User Widget Object References=@n3))
   > "SetText (Text)"(
    self=Get CommonTextBlock_201() -> (CommonTextBlock_201:Common Text Object Reference)
    InText="To Text (String)"(InString=Append!(A="Found ", B=@n4, C:String=" actors") -> (ReturnValue))
   )
   n5 = Construct Object from Class(execute=@n2.LoopBody, Class=/Game/Blueprints/Blutilities/ForTechArt/Data/TA_GenericListItemData.TA_GenericListItemData_C) -> (ReturnValue:TA Generic List Item Data Object Reference)
   n6 = Length(TargetArray:Array of Scene Component Object References=Get Components By Class!(self:Actor Object Reference=@n2, ComponentClass:Actor Component Class Reference=/Script/Engine.SceneComponent) -> (ReturnValue:Array of Scene Component Object References))
   > Set displayString!(
    execute
    displayString:String=Append(A=GetDisplayName(Object=@n2), B="=", C="To String (Integer)"(InInt=@n6))
    self:TA Generic List Item Data Object Reference=@n5
   ) -> (then, Output_Get:String)
   > Set actors!(
    execute
    actors:Array of Actor Object References=Make Array("[0]":Actor Object Reference=@n2) -> (Array:Array of Actor Object References)
    self:TA Generic List Item Data Object Reference=@n5
   ) -> (then, Output_Get:Array of Actor Object References)
   > addItem!(execute, self:TA Generic 01 List Object Reference=Get TA_Generic01_List(), dataObject:Object Reference=@n5) -> (then, newItem:User Widget Object Reference, newData:Object Reference)
   n7 = "On Item Right Clicked (TA_Generic01_List)"!() -> (OutputDelegate, then, isShiftDown:bool, isCtrlDown:bool, data:Object Reference, item:User Widget Object Reference) {is=title K2Node_ComponentBoundEvent}
   > n8 = Cast To TA_GenericListItemData!(execute, Object=@n7.data) -> (then, CastFailed, AsTA Generic List Item Data:TA Generic List Item Data Object Reference) {is=title K2Node_DynamicCast}
   > n9 = Branch(Condition=@n7.isShiftDown)
   n10 = Get actors!(self:TA Generic List Item Data Object Reference=@n8) -> (actors:Array of Actor Object References)
   n11 = GetSoftClassPath!(Class:Object Class Reference=Get Class(Object="Get (a copy)"(Array:Array of Actor Object References=@n10) -> (Output:Actor Object Reference))) -> (ReturnValue_PathString:String)
   > SyncBrowserToObjects(AssetPaths=Make Array("[0]":String=Split(SourceString=@n11, InStr=., SearchDir=FromEnd)) -> (Array:Array of Strings))
   Focus Actors in Viewport!(" ":Exec=@n9.else, actors:Array of Actor Object References=@n10, collidableOnly:bool, DoCameraMove:bool=true) -> ("  ":Exec, Loc:Vector=~, Rot:Rotator=~) {is=title K2Node_MacroInstance}
   > SetSelectedLevelActors(self=Get EditorActorSubsystem(), ActorsToSelect=@n10)
   > n12 = For Each Loop with Break!(
    Exec:Exec
    Array:Array of Scene Component Object References=Get Components By Class(self="Get (a copy)"(Array=@n10), ComponentClass=&SceneComponent)
    Break:Exec
   ) -> (LoopBody:Exec, Array Element:Scene Component Object Reference, Array Index:int, Completed:Exec)
   > n13 = GetEditorProperty!(execute:Exec, Object:Object Reference=@n12, PropertyName:Name=Mobility) -> (then:Exec, PropertyValue:EComponentMobility Enum, ReturnValue:bool)
   > PrintString(InString=Append(A=GetDisplayName(Object=@n12), B=" == ", C=Enum to String!(Enumerator:EComponentMobility Enum=@n13) -> (ReturnValue:String)))
   "On Clicked (btnFix)"!() -> (OutputDelegate, then, widget:User Widget Object Reference, event:Pointer Event Structure) {is=title K2Node_ComponentBoundEvent}
   > FixSimpleBuildingActors()
   "On Clicked (btnFix_1)"!() -> (OutputDelegate, then, widget:User Widget Object Reference, event:Pointer Event Structure) {is=title K2Node_ComponentBoundEvent}
   > ResetMobililtyOnAllComponents()
   n14 = For Each Loop(Array:Array of Static Mesh Actor Object References) -> (Array Element:Static Mesh Actor Object Reference)
   > Construct Object from Class(Class=&TA_GenericListItemData_C, self=@n14)
   "Not Equal (Enum)"!(A:EComponentMobility Enum=@n13, B:EComponentMobility Enum=Static) -> (ReturnValue:bool)
   Enum to String(Enumerator=@n13)
   "integer < integer"(A=@n6)
  ]}
 FindActorsWithMoveableComponents {nodes [
   FindActorsWithMoveableComponents() {is=title K2Node_FunctionEntry}
   > n1 = GetAllLevelActors(self=Get EditorActorSubsystem())
   > n2 = For Each Loop(Array:Array of Actor Object References=@n1) -> (Array Element:Actor Object Reference)
   Return Node!(execute=@n2.Completed, actors:Array of Actor Object References=Get l_actors() -> (l_actors:Array of Actor Object References))
   > n3 = For Each Loop with Break(
    Array:Array of Static Mesh Component Object References=Get Components By Class(self=@n2, ComponentClass=&StaticMeshComponent) -> (ReturnValue:Array of Static Mesh Component Object References)
    Break=@n6.then
   ) -> (Array Element:Static Mesh Component Object Reference)
   > n4 = Branch(Condition=Get bIsEditorOnly!(self:Actor Component Object Reference=@n3) -> (bIsEditorOnly:bool))
   n5 = GetEditorProperty(execute=@n4.else, Object=@n3, PropertyName=Mobility)
   > Branch(Condition=@n5.ReturnValue)
   > Branch(Condition="Not Equal (Enum)"(A=@n5, B=Static))
   > n6 = Add(TargetArray:Array of Actor Object References=Get l_actors(), NewItem:"Actor Object Reference (by ref)"=@n2) {is=title K2Node_CallArrayFunction}
  ]}
 FixSimpleBuildingActors {nodes [
   FixSimpleBuildingActors() {is=title K2Node_FunctionEntry}
   > n1 = FindActorsWithMoveableComponents() {is=title K2Node_CallFunction}
   > n2 = Filter by Class!(
    execute
    TargetArray:Array of Object References=@n1
    ObjectClass:Object Class Reference=/Script/FortniteGame.BuildingActor
    FilterType:EEditorScriptingFilterType Enum
   ) -> (then, ReturnValue:Array of Building Actor Object References)
   > n3 = For Each Loop(Array:Array of Building Actor Object References=@n2) -> (Array Element:Building Actor Object Reference)
   n4 = Length(TargetArray:Array of Building Actor Object References=Get l_actors() -> (l_actors:Array of Building Actor Object References))
   PrintString(execute=@n3.Completed, InString=Append!(A="Found ", B="To String (Integer)"(InInt=@n4)) -> (ReturnValue))
   n5 = Get Components By Class(self=@n3, ComponentClass=&SceneComponent) -> (ReturnValue:Array of Scene Component Object References)
   Branch(execute=@n3.LoopBody, Condition="integer < integer"(A=Length(TargetArray:Array of Scene Component Object References=@n5), B=4))
   n6 = Get RootComponent!(self:Actor Object Reference=@n3) -> (RootComponent:Scene Component Object Reference)
   > ResetEditorProperty!(execute, Object:Object Reference=@n6, PropertyName:Name=Mobility, ChangeNotifyMode:EPropertyAccessChangeNotifyMode Enum=Always) -> (then, ReturnValue:bool)
   Add(TargetArray:Array of Building Actor Object References=Get l_actors(), NewItem:"Building Actor Object Reference (by ref)"=@n3)
   SetEditorProperty(
    Object=@n6
    PropertyName=Mobility
    PropertyValue:EComponentMobility Enum=Literal enum EComponentMobility!(Enum:EComponentMobility Enum=Static) -> (ReturnValue:EComponentMobility Enum) {is=title K2Node_EnumLiteral}
   )
  ]}
 ResetMobililtyOnAllComponents {nodes [
   ResetMobililtyOnAllComponents() {is=title K2Node_FunctionEntry}
   > n1 = FindActorsWithMoveableComponents()
   > n2 = For Each Loop(Array:Array of Actor Object References=@n1) -> (Array Element:Actor Object Reference)
   > n3 = For Each Loop(Array:Array of Scene Component Object References=Get Components By Class(self=@n2, ComponentClass=&SceneComponent)) -> (Array Element:Scene Component Object Reference)
   > ResetEditorProperty(Object=@n3, PropertyName=Mobility, ChangeNotifyMode=Always)
  ]}
]
variables []
widgets {
 info {rootWidgetClass=&StackBox, widgetCount=7}
 widgets [
  StackBox_0 {widgetClassPath=&StackBox, widget=StackBox_0, properties {orientation=Orient_Vertical}}
  btnFind {
   widgetClassPath=&TA_Generic01_Button
   parent=StackBox_0
   widget=btnFind
   properties {label=Find Actors with Movable Components}
   slot {class=StackBoxSlot}
  }
  CommonTextBlock_201 {
   widgetClassPath=&CommonTextBlock
   bIsVariable=true
   parent=StackBox_0
   widget=CommonTextBlock_201
   properties {
    style=&TextStyle-Base_C
    text=" "
    font {
     bForceMonospaced=false
     bMaterialIsStencil=false
     fontMaterial=null
     fontObject=&NotoSans
     letterSpacing=0
     monospacedWidth=1
     size=24
     skewAmount=0
     typefaceFontName=NotoSans-Regular
     outlineSettings {
      bApplyOutlineToDropShadows=false
      bMiteredCorners=false
      bSeparateFillAlpha=false
      outlineMaterial=null
      outlineSize=0
      outlineColor=<rgba 0 0 0 1>
     }
    }
    shadowOffset=<0 0>
   }
   slot {class=StackBoxSlot}
  }
  TextBlock_81 {
   widgetClassPath=&TextBlock
   parent=StackBox_0
   widget=TextBlock_81
   properties {
    text="Right Click to focus in world\r\nShift Right Click to find in Content Browser"
    colorAndOpacity {colorUseRule=UseColor_Specified, specifiedColor=<rgba 1 1 1 0.461>}
    font {
     bForceMonospaced=false
     bMaterialIsStencil=false
     fontMaterial=null
     fontObject=&Roboto
     letterSpacing=0
     monospacedWidth=1
     size=12
     skewAmount=0
     typefaceFontName=Bold
     outlineSettings {
      bApplyOutlineToDropShadows=false
      bMiteredCorners=false
      bSeparateFillAlpha=false
      outlineMaterial=null
      outlineSize=0
      outlineColor=<rgba 0 0 0 1>
     }
    }
   }
   slot {class=StackBoxSlot}
  }
  TA_Generic01_List {
   widgetClassPath=/Game/Blueprints/Blutilities/ForTechArt/Widgets/List/TA_Generic01_List
   bIsVariable=true
   parent=StackBox_0
   widget=TA_Generic01_List
   properties {allowMultiSelect=false}
   slot {class=StackBoxSlot, properties {size {sizeRule=Fill, value=1}}}
  }
  btnFix {
   widgetClassPath=&TA_Generic01_Button
   parent=StackBox_0
   widget=btnFix
   properties {label=Fix Simple Building Actors}
   slot {class=StackBoxSlot, properties {padding=<ltrb 0 10 0 0>}}
  }
  btnFix_1 {
   widgetClassPath=&TA_Generic01_Button
   parent=StackBox_0
   widget=btnFix_1
   properties {label=Reset Mobility on All Actors' Components}
   slot {class=StackBoxSlot, properties {padding=<ltrb 0 10 0 0>}}
  }
 ]
}
